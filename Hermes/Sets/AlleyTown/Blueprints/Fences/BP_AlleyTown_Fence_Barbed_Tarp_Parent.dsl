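; Parent Blueprint for AlleyTown barbed fences with a tarp, which tints the tarp to a per-instance color.
; The Construction Script reads the material in slot 0 of TarpMesh and, if it is valid, creates a dynamic material instance from it.
; It then writes TarpColor into that instance's "Base Color Multiplier" vector parameter.
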(fn ConstructionScript
  (bind _tarpmesh (Variables|Default|GetTarpMesh))
  (bind _returnvalue (Rendering|Material|GetMaterial _tarpmesh 0))
  (if (Utilities|IsValid _returnvalue :!class /Script/Engine.KismetSystemLibrary)
    (bind _returnvalue_2 (Rendering|Material|CreateDynamicMaterialInstance _tarpmesh 0 _returnvalue "None" :!class /Script/Engine.PrimitiveComponent))
    (Rendering|Material|SetVectorParameterValue _returnvalue_2 "Base Color Multiplier" (Variables|Default|GetTarpColor) :!class /Script/Engine.MaterialInstanceDynamic)))
